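# Material function in the Athena landscape set that recolors a grayscale value by looking it up in a gradient texture.
# The Grayscale input interpolates between GradientMin and GradientMax, is clamped, and feeds the GradientTexture sample as a scalar UV, outputting RGB.
# Inputs fall back to preview values, including a flower texture's green channel and a flower gradient texture, suggesting use for foliage color variation.
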MaterialFunction /Game/Athena/Environments/Landscape/MaterialFunctions/MF_ColorGradientMap
graph {nodes [
  Constant() {r=0.5}
  n1 = LinearInterpolate(
   A=FunctionInput(Preview=Constant() {r=0.1}) {bUsePreviewValueAsDefault=true, inputName=GradientMin, inputType=FunctionInput_Scalar}
   B=FunctionInput(Preview=Constant() {r=1}) {bUsePreviewValueAsDefault=true, inputName=GradientMax, inputType=FunctionInput_Scalar, sortPriority=1}
   Alpha=FunctionInput(Preview=TextureSample() {samplerType=SAMPLERTYPE_LinearColor, texture=/Game/Athena/Environments/Landscape/Foliage/T_Athena_Flowers_01}.G) {bUsePreviewValueAsDefault=true, inputName=Grayscale, inputType=FunctionInput_Scalar, sortPriority=2}
  )
  n2 = TextureObject() {texture=/Game/Athena/Environments/Landscape/Foliage/T_Athena_FlowerGrad_03}
  FunctionOutput(None=TextureSample(
    UVs=Clamp(None=@n1) {minDefault=0.1}
    Tex=FunctionInput(Preview=@n2) {bUsePreviewValueAsDefault=true, inputName=GradientTexture, inputType=FunctionInput_Texture2D, sortPriority=3}
   ).RGB)
 ]}
